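# Unlit translucent particle sprite material from the tutorial assets, rendering animated smoke with a blackbody-driven glow.
# An 8x8 sub-UV smoke texture is added to a tiled soft-smoke texture, then multiplied by a BlackBody color whose temperature comes from a particle dynamic parameter.
# The tiled texture also blends particle alpha with a lowered copy, breaking up the sub-UV alpha as it fades before clamping.
# Responsive AA is enabled.
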Material /Engine/Tutorial/SubEditors/TutorialAssets/M_smoke_subUV_blackbody
graph {
 properties {
  bEnableResponsiveAA=true
  bUsedWithParticleSprites=true
  blendMode=BLEND_Translucent
  pixelDepthOffsetMode=PDOM_Legacy
  shadingModel=MSM_Unlit
  usedShadingModels=Unlit
 }
 nodes [
  n1 = ParticleSubUV!(UVs, Tex, Apply View MipBias) {texture=&/T_SmokeSubUV_8X8}
  n2 = ParticleColor()
  n3 = Add(A=Constant(), B=@n2.A)
  n4 = Add(A=Constant() {r=-0.5}, B=@n2.A)
  n5 = TextureSample(UVs=TextureCoordinate() {uTiling=8, vTiling=8}) {texture=&/T_soft_smoke}
  Clamp(None=Multiply(A=@n1.A, B=LinearInterpolate(A=@n3, B=@n4, Alpha=@n5.R)))
  Multiply(
   A=BlackBody!(Temp=DynamicParameter() {defaultValue=<rgba 0 0 0 1>, paramNames [temperature,Param2,Param3,Param4]}.temperature)
   B=Add(A=@n5.RGB, B=@n1.RGB)
  )
 ]
}
