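# Tutorial Animation Blueprint for the engine's SkeletalCube skeleton, serving as a minimal example in the editor's tutorial assets.
# The AnimGraph feeds the Cube_StateMachine straight into the output pose; its only state, Static, has no animation connected.
# The EventGraph holds an empty update event and an unwired TryGetPawnOwner node.
# SomeBoolean and SomeFloat are placeholder variables that no graph uses.
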AnimBlueprint /Engine/Tutorial/SubEditors/TutorialAssets/TutorialAnimationBlueprint
target_skeleton=/Engine/EngineMeshes/SkeletalCube_Skeleton
graphs [
 AnimGraph {nodes [
   AnimGraphNode_StateMachine_1 = Cube_StateMachine() {is=title AnimGraphNode_StateMachine}
   Output Pose(Result=@AnimGraphNode_StateMachine_1) {layerGroup=None}
  ]}
 Cube_StateMachine {
  entry=Static
  path=AnimGraph.AnimGraphNode_StateMachine_1.Cube_StateMachine
  ids=id
  nodes [AnimStateNode_1 = Static() {is=title AnimStateNode}]
 }
 Static {path=AnimGraph.AnimGraphNode_StateMachine_1.Cube_StateMachine.AnimStateNode_1.Static, nodes [Output Animation Pose()]}
 EventGraph {nodes [
   TryGetPawnOwner()
   Event BlueprintUpdateAnimation()
  ]}
]
variables [
 SomeBoolean {type=boolean}
 SomeFloat {type=number}
]
