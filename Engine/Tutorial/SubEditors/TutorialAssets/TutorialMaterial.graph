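# An empty placeholder Material in the engine's sub-editor tutorial assets.
# Its graph has no expressions and no connected properties, so it renders with default Default Lit surface values.
# Pixel depth offset uses the legacy mode.
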Material /Engine/Tutorial/SubEditors/TutorialAssets/TutorialMaterial
graph {expressions {}, properties {pixelDepthOffsetMode=PDOM_Legacy, usedShadingModels=Default Lit}}
